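# Sound cue for Carmine's third-person jump foley on dirt in Athena, routed to the enemy 3P player foley class with player jump attenuation.
# A Random node picks evenly among three Commando jump waves, and a Modulator varies volume and raises pitch above normal.
# It uses local and global footstep concurrency sets and carries owner-limited, single-voice override settings that prevent new instances.
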SoundCue /Figment_Martin/Athena/Sounds/Character/Footsteps/3P/Carmine/Athena_Carmine_Jump_Dirt_3P_Cue
root=Modulator_2
nodes [
 {id=Modulator_2, children [Random_0], properties {pitchMax=1.3, pitchMin=1.1, volumeMax=1, volumeMin=0.7}}
 {id=Random_0, children [WavePlayer_349, WavePlayer_352, WavePlayer_353], properties {bShouldExcludeFromBranchCulling=true, weights [1,1,1]}}
 {id=WavePlayer_349, properties {soundWaveAssetPtr=/Game/Sounds/Fortress_Foley/Commando/Commando_Jump_006}}
 {id=WavePlayer_352, properties {soundWaveAssetPtr=/Game/Sounds/Fortress_Foley/Commando/Commando_Jump_003}}
 {id=WavePlayer_353, properties {soundWaveAssetPtr=/Game/Sounds/Fortress_Foley/Commando/Commando_Jump_004}}
]
properties {
 attenuationSettings=/Game/Sounds/Attenuation/Fort_Player_Jump_3P_Athena_Att
 maxDistance=3584
 soundClassObject=&Fort_Player_Foley_3P_Enemy_Athena
 concurrencyOverrides {bLimitToOwner=true, maxCount=1, resolutionRule=PreventNew}
 concurrencySet [&Athena_Footstep_Local_Conc, &Athena_Footstep_Global_Conc]
}
