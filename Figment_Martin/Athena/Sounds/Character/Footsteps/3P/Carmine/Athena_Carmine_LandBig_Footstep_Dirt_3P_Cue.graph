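# Third-person sound cue for Carmine's big landing footstep on dirt, routed to the enemy foley sound class.
# Mixes two layers: a random normal-landing sample with footstep attenuation, and a random distant step sample, pitched down, using far weapon attenuation.
# Both layers reuse waves from the first-person Carmine folder.
# The cue overrides attenuation for long-range falloff and references the global footstep concurrency set.
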SoundCue /Figment_Martin/Athena/Sounds/Character/Footsteps/3P/Carmine/Athena_Carmine_LandBig_Footstep_Dirt_3P_Cue
root=Mixer_0
nodes [
 {
  id=Attenuation_0
  children [Random_1]
  properties {attenuationSettings=/Figment_Martin/Sounds/Attenuation/Fort_Player_Footsteps_Carmine_3P_Athena_Att}
 }
 {
  id=Attenuation_1
  children [Modulator_0]
  properties {attenuationSettings=/Figment_Martin/Sounds/Attenuation/Fort_Weapon_Carmine_Far_Athena_Att, attenuationOverrides {bAttenuate=false}}
 }
 {id=Mixer_0, children [Attenuation_0, Attenuation_1], properties {inputVolume [1,1]}}
 {id=Modulator_0, children [Random_0], properties {pitchMax=0.8, pitchMin=0.8}}
 {
  id=Random_0
  children [WavePlayer_0, WavePlayer_1, WavePlayer_17, WavePlayer_24, WavePlayer_25, WavePlayer_26, WavePlayer_27]
  properties {weights [1,1,1,1,1,1,1]}
 }
 {id=Random_1, children [WavePlayer_15, WavePlayer_16], properties {weights [1,1]}}
 {id=WavePlayer_0, properties {soundWaveAssetPtr=/Figment_Martin/Athena/Sounds/Character/Footsteps/1P/Carmine/Carmine_Step_03_distant}}
 {id=WavePlayer_1, properties {soundWaveAssetPtr=/Figment_Martin/Athena/Sounds/Character/Footsteps/1P/Carmine/Carmine_Step_04_distant}}
 {id=WavePlayer_15, properties {soundWaveAssetPtr=/Figment_Martin/Athena/Sounds/Character/Footsteps/1P/Carmine/Carmine_Land_Normal_01}}
 {id=WavePlayer_16, properties {soundWaveAssetPtr=/Figment_Martin/Athena/Sounds/Character/Footsteps/1P/Carmine/Carmine_Land_Normal_02}}
 {id=WavePlayer_17, properties {soundWaveAssetPtr=/Figment_Martin/Athena/Sounds/Character/Footsteps/1P/Carmine/Carmine_Step_05_distant}}
 {id=WavePlayer_24, properties {soundWaveAssetPtr=/Figment_Martin/Athena/Sounds/Character/Footsteps/1P/Carmine/Carmine_Step_06_distant}}
 {id=WavePlayer_25, properties {soundWaveAssetPtr=/Figment_Martin/Athena/Sounds/Character/Footsteps/1P/Carmine/Carmine_Step_07_distant}}
 {id=WavePlayer_26, properties {soundWaveAssetPtr=/Figment_Martin/Athena/Sounds/Character/Footsteps/1P/Carmine/Carmine_Step_08_distant}}
 {id=WavePlayer_27, properties {soundWaveAssetPtr=/Figment_Martin/Athena/Sounds/Character/Footsteps/1P/Carmine/Carmine_Step_09_distant}}
]
properties {
 bOverrideAttenuation=true
 maxDistance=30400
 soundClassObject=&Fort_Player_Foley_3P_Enemy_Athena
 volumeMultiplier=1
 attenuationOverrides {falloffDistance=30000}
 concurrencyOverrides {bLimitToOwner=true, maxCount=2, resolutionRule=StopOldest}
 concurrencySet [&Athena_Footstep_Global_Conc]
}
