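# Third-person sound cue for Carmine's landing footstep on dirt, combining a near layer and a distant layer with separate attenuation nodes.
# The near layer mixes a random landing sound with quiet, pitch- and volume-varied Commando foley under Carmine footstep attenuation.
# The far layer plays a random distant Carmine step, pitched down, through far weapon attenuation.
# The cue uses enemy foley sound class, long-range attenuation override and global footstep concurrency.
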SoundCue /Figment_Martin/Athena/Sounds/Character/Footsteps/3P/Carmine/Athena_Carmine_Land_Footstep_Dirt_3P_Cue
root=Mixer_0
nodes [
 {
  id=Attenuation_0
  children [Mixer_1]
  properties {attenuationSettings=/Figment_Martin/Sounds/Attenuation/Fort_Player_Footsteps_Carmine_3P_Athena_Att}
 }
 {
  id=Attenuation_1
  children [Modulator_0]
  properties {attenuationSettings=/Figment_Martin/Sounds/Attenuation/Fort_Weapon_Carmine_Far_Athena_Att, attenuationOverrides {bAttenuate=false}}
 }
 {id=Mixer_0, children [Attenuation_0, Attenuation_1], properties {inputVolume [1,1]}}
 {id=Mixer_1, children [Modulator_11,Modulator_2], properties {inputVolume [0.1,1]}}
 {id=Modulator_0, children [Random_2], properties {pitchMax=0.8, pitchMin=0.8}}
 {id=Modulator_11, children [Random_3], properties {pitchMax=1, pitchMin=0.7, volumeMax=1, volumeMin=0.3}}
 {id=Modulator_2, children [Random_1], properties {pitchMax=1, pitchMin=1, volumeMax=1, volumeMin=1}}
 {id=Random_1, children [WavePlayer_15, WavePlayer_16], properties {weights [1,1]}}
 {
  id=Random_2
  children [WavePlayer_2, WavePlayer_3, WavePlayer_17, WavePlayer_24, WavePlayer_25, WavePlayer_26, WavePlayer_27]
  properties {weights [1,1,1,1,1,1,1]}
 }
 {
  id=Random_3
  children [WavePlayer_6, WavePlayer_7, WavePlayer_8, WavePlayer_9, WavePlayer_10, WavePlayer_11]
  properties {bShouldExcludeFromBranchCulling=true, weights [1,1,1,1,1,1]}
 }
 {id=WavePlayer_10, properties {soundWaveAssetPtr=/Game/Sounds/Fortress_Foley/Commando/Commando_Foley_Stereo_005}}
 {id=WavePlayer_11, properties {soundWaveAssetPtr=/Game/Sounds/Fortress_Foley/Commando/Commando_Foley_Stereo_009}}
 {id=WavePlayer_15, properties {soundWaveAssetPtr=/Figment_Martin/Athena/Sounds/Character/Footsteps/1P/Carmine/Carmine_Land_Normal_01}}
 {id=WavePlayer_16, properties {soundWaveAssetPtr=/Figment_Martin/Athena/Sounds/Character/Footsteps/1P/Carmine/Carmine_Land_Normal_02}}
 {id=WavePlayer_17, properties {soundWaveAssetPtr=/Figment_Martin/Athena/Sounds/Character/Footsteps/1P/Carmine/Carmine_Step_05_distant}}
 {id=WavePlayer_2, properties {soundWaveAssetPtr=/Figment_Martin/Athena/Sounds/Character/Footsteps/1P/Carmine/Carmine_Step_03_distant}}
 {id=WavePlayer_24, properties {soundWaveAssetPtr=/Figment_Martin/Athena/Sounds/Character/Footsteps/1P/Carmine/Carmine_Step_06_distant}}
 {id=WavePlayer_25, properties {soundWaveAssetPtr=/Figment_Martin/Athena/Sounds/Character/Footsteps/1P/Carmine/Carmine_Step_07_distant}}
 {id=WavePlayer_26, properties {soundWaveAssetPtr=/Figment_Martin/Athena/Sounds/Character/Footsteps/1P/Carmine/Carmine_Step_08_distant}}
 {id=WavePlayer_27, properties {soundWaveAssetPtr=/Figment_Martin/Athena/Sounds/Character/Footsteps/1P/Carmine/Carmine_Step_09_distant}}
 {id=WavePlayer_3, properties {soundWaveAssetPtr=/Figment_Martin/Athena/Sounds/Character/Footsteps/1P/Carmine/Carmine_Step_04_distant}}
 {id=WavePlayer_6, properties {soundWaveAssetPtr=/Game/Sounds/Fortress_Foley/Commando/Commando_Foley_Stereo_006}}
 {id=WavePlayer_7, properties {soundWaveAssetPtr=/Game/Sounds/Fortress_Foley/Commando/Commando_Foley_Stereo_008}}
 {id=WavePlayer_8, properties {soundWaveAssetPtr=/Game/Sounds/Fortress_Foley/Commando/Commando_Foley_Stereo_001}}
 {id=WavePlayer_9, properties {soundWaveAssetPtr=/Game/Sounds/Fortress_Foley/Commando/Commando_Foley_Stereo_003}}
]
properties {
 attenuationSettings=/Game/Sounds/Attenuation/Fort_Player_Land_3P_Athena_Att
 bOverrideAttenuation=true
 maxDistance=30000
 soundClassObject=&Fort_Player_Foley_3P_Enemy_Athena
 volumeMultiplier=1
 attenuationOverrides {falloffDistance=30000, attenuationShapeExtents {x=0}}
 concurrencyOverrides {bLimitToOwner=true, maxCount=1, resolutionRule=StopOldest}
 concurrencySet [&Athena_Footstep_Global_Conc]
}
