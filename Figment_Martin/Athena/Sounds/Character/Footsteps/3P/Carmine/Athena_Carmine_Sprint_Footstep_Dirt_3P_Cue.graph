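# Third-person Carmine sprint footstep cue for dirt, built from step waves in the 1P folder.
# Two random pools, close steps and distant variants, each pass through their own attenuation asset and a weapon-fire 3P sound class override.
# A mixer blends them with the distant layer quieter, then a modulator applies a fixed pitch raise.
# The cue uses long-range attenuation and local and global footstep concurrency sets.
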SoundCue /Figment_Martin/Athena/Sounds/Character/Footsteps/3P/Carmine/Athena_Carmine_Sprint_Footstep_Dirt_3P_Cue
root=Modulator_2
nodes [
 {
  id=Attenuation_2
  children [Random_1]
  properties {attenuationSettings=/Figment_Martin/Sounds/Attenuation/Fort_Player_Footsteps_Carmine_3P_Athena_Att, attenuationOverrides {bSpatialize=false}}
 }
 {
  id=Attenuation_3
  children [Random_2]
  properties {attenuationSettings=/Figment_Martin/Sounds/Attenuation/Fort_Weapon_Carmine_Medium_Athena_Att, attenuationOverrides {bSpatialize=false}}
 }
 {id=Mixer_1, children [SoundClass_0,SoundClass_1], properties {inputVolume [1,0.35]}}
 {id=Modulator_2, children [Mixer_1], properties {pitchMax=1.1, pitchMin=1.1, volumeMax=1, volumeMin=1}}
 {
  id=Random_1
  children [WavePlayer_9, WavePlayer_10, WavePlayer_11, WavePlayer_12, WavePlayer_13, WavePlayer_14, WavePlayer_15, WavePlayer_16, WavePlayer_17]
  properties {weights [1,1,1,1,1,1,1,1,1]}
 }
 {
  id=Random_2
  children [WavePlayer_18, WavePlayer_19, WavePlayer_20, WavePlayer_24, WavePlayer_25, WavePlayer_26, WavePlayer_27]
  properties {weights [1,1,1,1,1,1,1]}
 }
 {
  id=SoundClass_0
  children [Attenuation_2]
  properties {soundClassOverride=&Fort_WeaponRanged_Fire_3P_PSM}
 }
 {
  id=SoundClass_1
  children [Attenuation_3]
  properties {soundClassOverride=&Fort_WeaponRanged_Fire_3P_PSM}
 }
 {id=WavePlayer_10, properties {soundWaveAssetPtr=/Figment_Martin/Athena/Sounds/Character/Footsteps/1P/Carmine/Carmine_Step_02}}
 {id=WavePlayer_11, properties {soundWaveAssetPtr=/Figment_Martin/Athena/Sounds/Character/Footsteps/1P/Carmine/Carmine_Step_03}}
 {id=WavePlayer_12, properties {soundWaveAssetPtr=/Figment_Martin/Athena/Sounds/Character/Footsteps/1P/Carmine/Carmine_Step_04}}
 {id=WavePlayer_13, properties {soundWaveAssetPtr=/Figment_Martin/Athena/Sounds/Character/Footsteps/1P/Carmine/Carmine_Step_05}}
 {id=WavePlayer_14, properties {soundWaveAssetPtr=/Figment_Martin/Athena/Sounds/Character/Footsteps/1P/Carmine/Carmine_Step_06}}
 {id=WavePlayer_15, properties {soundWaveAssetPtr=/Figment_Martin/Athena/Sounds/Character/Footsteps/1P/Carmine/Carmine_Step_07}}
 {id=WavePlayer_16, properties {soundWaveAssetPtr=/Figment_Martin/Athena/Sounds/Character/Footsteps/1P/Carmine/Carmine_Step_08}}
 {id=WavePlayer_17, properties {soundWaveAssetPtr=/Figment_Martin/Athena/Sounds/Character/Footsteps/1P/Carmine/Carmine_Step_09}}
 {id=WavePlayer_18, properties {soundWaveAssetPtr=/Figment_Martin/Athena/Sounds/Character/Footsteps/1P/Carmine/Carmine_Step_03_distant}}
 {id=WavePlayer_19, properties {soundWaveAssetPtr=/Figment_Martin/Athena/Sounds/Character/Footsteps/1P/Carmine/Carmine_Step_04_distant}}
 {id=WavePlayer_20, properties {soundWaveAssetPtr=/Figment_Martin/Athena/Sounds/Character/Footsteps/1P/Carmine/Carmine_Step_05_distant}}
 {id=WavePlayer_24, properties {soundWaveAssetPtr=/Figment_Martin/Athena/Sounds/Character/Footsteps/1P/Carmine/Carmine_Step_06_distant}}
 {id=WavePlayer_25, properties {soundWaveAssetPtr=/Figment_Martin/Athena/Sounds/Character/Footsteps/1P/Carmine/Carmine_Step_07_distant}}
 {id=WavePlayer_26, properties {soundWaveAssetPtr=/Figment_Martin/Athena/Sounds/Character/Footsteps/1P/Carmine/Carmine_Step_08_distant}}
 {id=WavePlayer_27, properties {soundWaveAssetPtr=/Figment_Martin/Athena/Sounds/Character/Footsteps/1P/Carmine/Carmine_Step_09_distant}}
 {id=WavePlayer_9, properties {soundWaveAssetPtr=/Figment_Martin/Athena/Sounds/Character/Footsteps/1P/Carmine/Carmine_Step_01}}
]
properties {
 bOverrideAttenuation=true
 maxDistance=30000
 volumeMultiplier=1
 attenuationOverrides {falloffDistance=30000, attenuationShapeExtents {x=0}}
 concurrencyOverrides {bLimitToOwner=true, maxCount=1, resolutionRule=StopOldest}
 concurrencySet [&Athena_Footstep_Local_Conc, &Athena_Footstep_Global_Conc]
}
