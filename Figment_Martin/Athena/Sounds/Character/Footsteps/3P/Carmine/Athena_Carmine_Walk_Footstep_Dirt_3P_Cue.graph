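# Third-person sound cue for Carmine's walking footsteps on dirt, built from wave files in the 1P Carmine footstep folder.
# A mixer blends a random close-step layer with a quieter random distant-step layer, each with its own attenuation asset, footstep and medium weapon respectively.
# Both layers override their sound class to third-person ranged weapon fire, while the cue uses enemy foley.
# Local and global footstep concurrency sets are assigned.
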SoundCue /Figment_Martin/Athena/Sounds/Character/Footsteps/3P/Carmine/Athena_Carmine_Walk_Footstep_Dirt_3P_Cue
root=Mixer_2
nodes [
 {
  id=Attenuation_2
  children [Random_0]
  properties {attenuationSettings=/Figment_Martin/Sounds/Attenuation/Fort_Player_Footsteps_Carmine_3P_Athena_Att, attenuationOverrides {bSpatialize=false}}
 }
 {
  id=Attenuation_3
  children [Random_1]
  properties {attenuationSettings=/Figment_Martin/Sounds/Attenuation/Fort_Weapon_Carmine_Medium_Athena_Att, attenuationOverrides {bSpatialize=false}}
 }
 {id=Mixer_2, children [SoundClass_0,SoundClass_1], properties {inputVolume [1,0.25]}}
 {
  id=Random_0
  children [WavePlayer_0, WavePlayer_1, WavePlayer_2, WavePlayer_3, WavePlayer_4, WavePlayer_5, WavePlayer_12, WavePlayer_13, WavePlayer_14]
  properties {weights [1,1,1,1,1,1,1,1,1]}
 }
 {
  id=Random_1
  children [WavePlayer_15, WavePlayer_16, WavePlayer_17, WavePlayer_24, WavePlayer_25, WavePlayer_26, WavePlayer_27]
  properties {weights [1,1,1,1,1,1,1]}
 }
 {
  id=SoundClass_0
  children [Attenuation_2]
  properties {soundClassOverride=&Fort_WeaponRanged_Fire_3P_PSM}
 }
 {
  id=SoundClass_1
  children [Attenuation_3]
  properties {soundClassOverride=&Fort_WeaponRanged_Fire_3P_PSM}
 }
 {id=WavePlayer_0, properties {soundWaveAssetPtr=/Figment_Martin/Athena/Sounds/Character/Footsteps/1P/Carmine/Carmine_Step_01}}
 {id=WavePlayer_1, properties {soundWaveAssetPtr=/Figment_Martin/Athena/Sounds/Character/Footsteps/1P/Carmine/Carmine_Step_02}}
 {id=WavePlayer_12, properties {soundWaveAssetPtr=/Figment_Martin/Athena/Sounds/Character/Footsteps/1P/Carmine/Carmine_Step_07}}
 {id=WavePlayer_13, properties {soundWaveAssetPtr=/Figment_Martin/Athena/Sounds/Character/Footsteps/1P/Carmine/Carmine_Step_08}}
 {id=WavePlayer_14, properties {soundWaveAssetPtr=/Figment_Martin/Athena/Sounds/Character/Footsteps/1P/Carmine/Carmine_Step_09}}
 {id=WavePlayer_15, properties {soundWaveAssetPtr=/Figment_Martin/Athena/Sounds/Character/Footsteps/1P/Carmine/Carmine_Step_03_distant}}
 {id=WavePlayer_16, properties {soundWaveAssetPtr=/Figment_Martin/Athena/Sounds/Character/Footsteps/1P/Carmine/Carmine_Step_04_distant}}
 {id=WavePlayer_17, properties {soundWaveAssetPtr=/Figment_Martin/Athena/Sounds/Character/Footsteps/1P/Carmine/Carmine_Step_05_distant}}
 {id=WavePlayer_2, properties {soundWaveAssetPtr=/Figment_Martin/Athena/Sounds/Character/Footsteps/1P/Carmine/Carmine_Step_03}}
 {id=WavePlayer_24, properties {soundWaveAssetPtr=/Figment_Martin/Athena/Sounds/Character/Footsteps/1P/Carmine/Carmine_Step_06_distant}}
 {id=WavePlayer_25, properties {soundWaveAssetPtr=/Figment_Martin/Athena/Sounds/Character/Footsteps/1P/Carmine/Carmine_Step_07_distant}}
 {id=WavePlayer_26, properties {soundWaveAssetPtr=/Figment_Martin/Athena/Sounds/Character/Footsteps/1P/Carmine/Carmine_Step_08_distant}}
 {id=WavePlayer_27, properties {soundWaveAssetPtr=/Figment_Martin/Athena/Sounds/Character/Footsteps/1P/Carmine/Carmine_Step_09_distant}}
 {id=WavePlayer_3, properties {soundWaveAssetPtr=/Figment_Martin/Athena/Sounds/Character/Footsteps/1P/Carmine/Carmine_Step_04}}
 {id=WavePlayer_4, properties {soundWaveAssetPtr=/Figment_Martin/Athena/Sounds/Character/Footsteps/1P/Carmine/Carmine_Step_05}}
 {id=WavePlayer_5, properties {soundWaveAssetPtr=/Figment_Martin/Athena/Sounds/Character/Footsteps/1P/Carmine/Carmine_Step_06}}
]
properties {
 bOverrideAttenuation=true
 maxDistance=30400
 soundClassObject=&Fort_Player_Foley_3P_Enemy_Athena
 volumeMultiplier=1
 attenuationOverrides {falloffDistance=30000}
 concurrencyOverrides {bLimitToOwner=true, maxCount=1, resolutionRule=StopOldest}
 concurrencySet [&Athena_Footstep_Local_Conc, &Athena_Footstep_Global_Conc]
}
